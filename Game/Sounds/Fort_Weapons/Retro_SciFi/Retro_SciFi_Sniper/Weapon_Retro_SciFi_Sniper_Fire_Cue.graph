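# Sound cue for the Retro SciFi sniper's fire sound, layering close, medium and distant versions of the shot.
# A mixer combines a modulated single close sample in a third-person fire sound class with a distance crossfade in a distant-fire class.
# The crossfade blends overlapping medium and far bands, each choosing from three random variants with its own attenuation; only the medium band is modulated.
# Concurrency is overridden to prevent new instances at the limit.
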SoundCue /Game/Sounds/Fort_Weapons/Retro_SciFi/Retro_SciFi_Sniper/Weapon_Retro_SciFi_Sniper_Fire_Cue
root=Mixer_15
nodes [
 {id=Attenuation_80, children [Modulator_0], properties {attenuationSettings=/Game/Sounds/Attenuation/Fort_Weapon_Rifle_Sml_Att}}
 {id=Attenuation_81, children [Modulator_13], properties {attenuationSettings=/Game/Sounds/Attenuation/Fort_Weapon_Projectile_Medium_StW_Att}}
 {id=Attenuation_82, children [Random_2], properties {attenuationSettings=/Game/Sounds/Attenuation/Fort_Weapon_Projectile_Distant_StW_Att}}
 {
  id=DistanceCrossFade_25
  children [Attenuation_81, Attenuation_82]
  properties {crossFadeInput [
    {fadeInDistanceEnd=100, fadeInDistanceStart=0, fadeOutDistanceEnd=3000, fadeOutDistanceStart=2000, volume=1}
    {fadeInDistanceEnd=3000, fadeInDistanceStart=1000, fadeOutDistanceEnd=12000, fadeOutDistanceStart=10000, volume=1}
   ]}
 }
 {id=Mixer_15, children [SoundClass_55, SoundClass_56], properties {inputVolume [1,1]}}
 {id=Modulator_0, children [WavePlayer_1], properties {pitchMin=1}}
 {id=Modulator_13, children [Random_1]}
 {id=Random_1, children [WavePlayer_2,WavePlayer_3,WavePlayer_4], properties {weights [1,1,1]}}
 {id=Random_2, children [WavePlayer_5,WavePlayer_6,WavePlayer_7], properties {weights [1,1,1]}}
 {
  id=SoundClass_55
  children [Attenuation_80]
  properties {soundClassOverride=&Fort_WeaponRanged_Fire_3P_PSM}
 }
 {
  id=SoundClass_56
  children [DistanceCrossFade_25]
  properties {soundClassOverride=&Fort_WeaponRanged_Fire_Distant}
 }
 {id=WavePlayer_1, properties {soundWaveAssetPtr=&/Retro_SciFi_Sniper_Fire_Close_01}}
 {id=WavePlayer_2, properties {soundWaveAssetPtr=&/Retro_SciFi_Sniper_Fire_Med_03}}
 {id=WavePlayer_3, properties {soundWaveAssetPtr=&/Retro_SciFi_Sniper_Fire_Med_01}}
 {id=WavePlayer_4, properties {soundWaveAssetPtr=&/Retro_SciFi_Sniper_Fire_Med_02}}
 {id=WavePlayer_5, properties {soundWaveAssetPtr=&/Retro_SciFi_Sniper_Fire_Dist_03}}
 {id=WavePlayer_6, properties {soundWaveAssetPtr=&/Retro_SciFi_Sniper_Fire_Dist_01}}
 {id=WavePlayer_7, properties {soundWaveAssetPtr=&/Retro_SciFi_Sniper_Fire_Dist_02}}
]
properties {
 bOverrideConcurrency=true
 maxDistance=20000
 volumeMultiplier=0.7
 attenuationOverrides {distanceAlgorithm=NaturalSound, falloffDistance=4000, attenuationShapeExtents {x=1000}}
 concurrencyOverrides {resolutionRule=PreventNew}
}
